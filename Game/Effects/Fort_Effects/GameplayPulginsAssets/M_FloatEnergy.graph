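# Translucent unlit Niagara sprite and ribbon material that draws a distorted plasma energy effect tinted by particle color.
# Soft noise warps the plasma texture, both timed by a dynamic Time parameter, and the refined plasma drives emissive.
# Plasma also distorts glow-burst mask UVs, weighted by a vertical power falloff.
# That falloff, a UV clamp mask, particle alpha and depth fade shape opacity.
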Material /Game/Effects/Fort_Effects/GameplayPulginsAssets/M_FloatEnergy
graph {
 properties {
  bAutomaticallySetUsageInEditor=false
  bUsedWithNiagaraRibbons=true
  bUsedWithNiagaraSprites=true
  blendMode=BLEND_Translucent
  shadingModel=MSM_Unlit
  usedShadingModels=Unlit
 }
 nodes [
  n1 = ParticleColor()
  n2 = VectorParameter() {parameterName=gr34h656, defaultValue=<rgba 0 0.2 0 1>}
  n3 = VectorParameter() {group=Mask, parameterName=Mask_UVControl, channelNames {a=VPan, b=UPan, g=VScale, r=UScale}, defaultValue=<rgba 1 1.2 0 0>}
  n4 = MaterialFunctionCall(Float2=TextureCoordinate()) {materialFunction=&BreakOutFloat2Components}
  n5 = MaterialFunctionCall!(
   Texture
   AdditionalUVOffset=AppendVector(A=@n2.R, B=@n2.R)
   UVControl_RGB=@n3
   UVControl_A=@n3
   "[Override]TimeInput"
   "[Override]ScaleUVByCenter"=StaticBool() {value=true}
   "[Override]UVInput"=MaterialFunctionCall(X=@n4.R, Y=OneMinus(None=@n4.G)) {materialFunction=&MakeFloat2}.Result
  ) {materialFunction=/Game/Effects/Niagara/MasterMaterials/Functions/Common/FXCommon_TextureSample_V2}
  n6 = VectorParameter() {parameterName=zsc, defaultValue=<rgba 1 2 0 1>}
  n7 = Multiply(
   A=OneMinus(None=MaterialFunctionCall(Float2=@n5.UV) {materialFunction=&BreakOutFloat2Components}.G)
   B=@n6.R
  )
  n8 = Power(Base=OneMinus(None=Saturate(None=@n7)), Exp=@n6.G)
  n9 = TextureObjectParameter() {
   group=MainTexture
   parameterName=MainTexture_Texture
   samplerType=SAMPLERTYPE_LinearColor
   sortPriority=0
   texture=/Game/Effects/Fort_Effects/Emotes/UltraLight/Texture/T_Plasma_A_C
  }
  n10 = TextureObjectParameter() {
   group=MainNoise
   parameterName=MainNoise_Texture
   sortPriority=0
   texture=/Game/Effects/Fort_Effects/Emotes/UltraLight/Texture/T_SoftNoise_BC7
  }
  n11 = VectorParameter() {group=MainNoise, parameterName=MainNoise_UVControl, channelNames {a=VPan, b=UPan, g=VScale, r=UScale}, defaultValue=<rgba 2 1.3 0 0.3>}
  n12 = DynamicParameter() {paramNames [Time,Param2,Param3,Param4]}
  n13 = VectorParameter() {
   group=MainNoise
   parameterName=MainNoise_NoiseControl
   sortPriority=60
   channelNames {a=VStrength, b=UStrength, g=VOffset, r=UOffset}
   defaultValue=<rgba -0.5 -0.5 0.2 0.1>
  }
  n14 = VectorParameter() {
   group=MainTexture
   parameterName=MainTexture_UVControl
   channelNames {a=VPan, b=UPan, g=VScale, r=UScale}
   defaultValue=<rgba 0.5 0.1 0 0.15>
  }
  n15 = MaterialFunctionCall(
   Texture=@n9
   AdditionalUVOffset=MaterialFunctionCall!(
    NoiseValue=MaterialFunctionCall(Texture=@n10, UVControl_RGB=@n11, UVControl_A=@n11, "[Override]TimeInput"=@n12.Time) {materialFunction=&FXCommon_TextureSample}.RG
    Control_RGB=@n13
    Control_A=@n13
   ) {materialFunction=/Game/Effects/Niagara/MasterMaterials/Functions/Common/FXCommon_NoiseControl}.NoiseValue
   UVControl_RGB=@n14
   UVControl_A=@n14
   "[Override]TimeInput"=@n12.Time
  ) {materialFunction=&FXCommon_TextureSample}
  n16 = VectorParameter() {
   group=MainTextureToN
   parameterName=MainTextureToN_NoiseControl
   sortPriority=60
   channelNames {a=VStrength, b=UStrength, g=VOffset, r=UOffset}
   defaultValue=<rgba -0.5 -0.3 0 -0.7>
  }
  n17 = Add(
   A=@n5.UV
   B=Multiply(
    A=@n8
    B=MaterialFunctionCall(NoiseValue=@n15.R, Control_RGB=@n16, Control_A=@n16) {materialFunction=&FXCommon_NoiseControl}.NoiseValue
   )
  )
  n18 = Multiply(
   A=MaterialFunctionCall(
    Texture=TextureObjectParameter() {group=Mask, parameterName=Mask_Texture, sortPriority=0, texture=/Game/Effects/Fort_Effects/Textures/Mask/T_GlowBurst}
    "[Override]UVInput"=@n17
   ) {materialFunction=&FXCommon_TextureSample_V2}.R
   B=MaterialFunctionCall!(TexCoord=@n17, ClampChannel=Constant2Vector() {g=1, r=1}) {materialFunction=/Game/Effects/Fort_Effects/MaterialFunctions/MF_FX_TilingMethod}.Result
  )
  DepthFade(Opacity=Multiply(A=@n1.A, B=Multiply(A=@n18, B=@n8)), FadeDistance=ScalarParameter() {defaultValue=100, parameterName=FD})
  n19 = VectorParameter() {parameterName=CR, defaultValue=<rgba 0.5 5 1 0.5>}
  MaterialFunctionCall(Emissive=Multiply(
    A=MaterialFunctionCall(Texture=@n15.R, RGB=@n19.RGB, A=@n19.RGB) {materialFunction=&MF_FX_Refine}.Result
    B=@n1.RGB
   )) {materialFunction=&MF_FX_EmissiveAdjustments}
 ]
}
